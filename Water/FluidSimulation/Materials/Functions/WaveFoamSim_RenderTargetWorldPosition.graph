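# Material function in the Water library that converts a UV on the wave and foam simulation render target into a 2D world position.
# It centers the UV (defaulting to TexCoord) around zero, scales it by WaveCaptureSize and adds the XY of SimLocation.
# Both values come from the MPC_Water parameter collection, so the mapping follows the simulation's capture area.
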MaterialFunction /Water/FluidSimulation/Materials/Functions/WaveFoamSim_RenderTargetWorldPosition
graph {
 properties {libraryCategoriesText [Water]}
 nodes [
  n1 = TextureCoordinate()
  n2 = Subtract(A=ComponentMask(None=FunctionInput(Preview=@n1) {bUsePreviewValueAsDefault=true, inputName=UV, inputType=FunctionInput_Vector2}) {g=true, r=true}) {constB=0.5}
  n3 = CollectionParameter() {collection=/Water/Materials/MPC/MPC_Water, parameterName=WaveCaptureSize}
  n4 = CollectionParameter() {collection=&MPC_Water, parameterName=SimLocation}
  FunctionOutput(None=Add(A=Multiply(A=@n2, B=@n3), B=ComponentMask(None=@n4) {g=true, r=true})) {outputName=WorldPosition}
 ]
}
