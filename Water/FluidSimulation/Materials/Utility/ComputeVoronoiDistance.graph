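# Unlit utility material for the water fluid simulation that turns a Voronoi nearest-seed render target into a signed distance field.
# Point-sampled RT coordinates give UV distance and normalized direction, both negated where captured ground height is above water.
# Water height blends between WaterZ and WaterBodyData via UseTerrainSystem.
# It packs signed distance, ground height and biased direction into four channels, with direction Y inverted for alpha-composite output.
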Material /Water/FluidSimulation/Materials/Utility/ComputeVoronoiDistance
graph {
 properties {
  bAllowNegativeEmissiveColor=true
  blendMode=BLEND_AlphaComposite
  pixelDepthOffsetMode=PDOM_Legacy
  shadingModel=MSM_Unlit
  usedShadingModels=Unlit
 }
 nodes [
  n1 = MaterialFunctionCall!(
   UVs=TextureCoordinate()
   Texture Size=TextureProperty(Tex=TextureObjectParameter() {parameterName=RT, texture=&Black})
  ) {materialFunction=/Engine/Functions/Engine_MaterialFunctions01/Texturing/PointSampledUVs}
  n2 = TextureSampleParameter2D!(UVs=@n1.Result, Level=Constant(), Apply View MipBias) {
   mipValueMode=TMVM_MipLevel
   parameterName=RT
   samplerType=SAMPLERTYPE_LinearColor
   texture=/Engine/Functions/Engine_MaterialFunctions02/PivotPainter2/Black_1x1_EXR_Texture
  }
  n3 = ComponentMask(None=@n2.RGB) {g=true, r=true}
  n4 = TextureSampleParameter2D() {
   mipValueMode=TMVM_MipLevel
   parameterName=DepthRT
   samplerType=SAMPLERTYPE_LinearColor
   texture=&Black_1x1_EXR_Texture
  }
  n5 = Subtract(A=ScalarParameter() {defaultValue=50000, parameterName=CaptureZ}, B=@n4.R)
  n6 = MaterialFunctionCall!(UV) {materialFunction=/Water/FluidSimulation/Materials/Functions/WaveFoamSim_RenderTargetWorldPosition}
  n7 = Subtract(
   A=@n5
   B=LinearInterpolate(
    A=ScalarParameter() {defaultValue=-1536, parameterName=WaterZ}
    B=MaterialFunctionCall!(Position=@n6.WorldPosition) {materialFunction=/Water/Materials/Functions/WaterBodyData}.Water Z Height
    Alpha=ScalarParameter() {parameterName=UseTerrainSystem}
   )
  )
  n8 = LinearInterpolate(Alpha=Saturate(None=Ceil(None=@n7))) {constA=1, constB=-1}
  n9 = AppendVector(A=Subtract(A=@n1.Result, B=@n3), B=Constant())
  n10 = ComponentMask(None=MaterialFunctionCall(Vector=Multiply(A=@n9) {constB=500}) {materialFunction=&SafeNormalize}.Result) {g=true, r=true}
  n11 = MaterialFunctionCall(Float2=ConstantBiasScale(None=Multiply(A=@n10, B=@n8))) {materialFunction=&BreakOutFloat2Components}
  MaterialFunctionCall(X=Multiply(A=Distance(A=@n1.Result, B=@n3), B=@n8), Y=@n5, Z=@n11.R) {materialFunction=&MakeFloat3}
  OneMinus(None=@n11.G)
 ]
}
