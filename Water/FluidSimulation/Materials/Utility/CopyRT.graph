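# CopyRT is an unlit, alpha-composite utility material for water fluid simulation that copies RT texture channels and allows negative values.
# The Pass2 scalar lerps color from G placed in blue to R and G passed through with zero blue.
# It also lerps an alpha term from one minus saturated R to zero, so under alpha compositing the second pass is purely additive.
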Material /Water/FluidSimulation/Materials/Utility/CopyRT
graph {
 properties {
  bAllowNegativeEmissiveColor=true
  blendMode=BLEND_AlphaComposite
  pixelDepthOffsetMode=PDOM_Legacy
  shadingModel=MSM_Unlit
  usedShadingModels=Unlit
 }
 nodes [
  Constant()
  n1 = Constant()
  n2 = TextureSampleParameter2D() {
   parameterName=RT
   samplerType=SAMPLERTYPE_LinearColor
   texture=/Engine/Functions/Engine_MaterialFunctions02/PivotPainter2/Black_1x1_EXR_Texture
  }
  n3 = TextureSampleParameter2D() {
   parameterName=RT
   samplerType=SAMPLERTYPE_LinearColor
   texture=&Black_1x1_EXR_Texture
  }
  LinearInterpolate(
   A=MaterialFunctionCall(X=@n1, Y=@n1, Z=@n2.G) {materialFunction=&MakeFloat3}.Result
   B=AppendVector(A=ComponentMask(None=@n3.RGB) {g=true, r=true}, B=@n1)
   Alpha=ScalarParameter() {parameterName=Pass2}
  )
  LinearInterpolate(A=OneMinus(None=Saturate(None=@n2.R)), Alpha=ScalarParameter() {parameterName=Pass2}) {constB=0}
 ]
}
