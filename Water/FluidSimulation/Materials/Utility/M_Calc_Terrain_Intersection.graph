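# Unlit utility material in the water fluid simulation that calculates where waves meet terrain, using wave and landscape render targets.
# Water depth is wave height plus Wave Z minus landscape height, compared across current and previous frames to mask falling, shallow water.
# A second branch samples at a V-shifted UV, combining a depth-band interface mask with wave slope along V.
# No final output connection is shown, and several branches end unused.
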Material /FortWater/FluidSimulation/Materials/Utility/M_Calc_Terrain_Intersection
graph {
 properties {bAllowNegativeEmissiveColor=true, pixelDepthOffsetMode=PDOM_Legacy, shadingModel=MSM_Unlit, usedShadingModels=Unlit}
 nodes [
  Multiply()
  n1 = TextureSampleParameter2D() {parameterName=PreviousWaveRT, samplerType=SAMPLERTYPE_LinearColor, texture=/FortWater/FluidSimulation/Textures/RT_B}
  n2 = TextureSampleParameter2D() {
   parameterName=LandscapeRT
   samplerType=SAMPLERTYPE_LinearColor
   texture=/Engine/Functions/Engine_MaterialFunctions02/PivotPainter2/Black_1x1_EXR_Texture
  }
  n3 = Subtract(A=Add(A=@n1.R, B=ScalarParameter() {defaultValue=2000, parameterName=Wave Z}), B=@n2.R) {desc=Water Height Last Frame}
  Multiply(A=@n3) {constB=-1}
  n4 = TextureSampleParameter2D() {parameterName=WaveRT, samplerType=SAMPLERTYPE_LinearColor, texture=/FortWater/FluidSimulation/Textures/RT_A}
  n5 = Subtract(A=Add(A=@n4.R, B=ScalarParameter() {defaultValue=2000, parameterName=Wave Z}), B=@n2.R) {desc=Water Height This Frame}
  n6 = Saturate(None=Divide(A=@n5, B=ScalarParameter() {defaultValue=500, parameterName=Max Depth for Water Reads}))
  n7 = Subtract(A=@n3, B=@n5)
  Multiply(A=Multiply(A=OneMinus(None=@n6), B=Max(A=@n7) {constB=0}), B=Saturate(None=Ceil(None=@n3)))
  n8 = Add(A=TextureCoordinate(), B=Multiply(A=Constant2Vector() {g=1}, B=ScalarParameter() {parameterName=WaveShift}))
  n9 = TextureSampleParameter2D(UVs=@n8) {desc=Where Wave is Now, parameterName=WaveRT, samplerType=SAMPLERTYPE_LinearColor, texture=&RT_A}
  n10 = TextureSampleParameter2D(UVs=@n8) {
   parameterName=LandscapeRT
   samplerType=SAMPLERTYPE_LinearColor
   texture=&Black_1x1_EXR_Texture
  }
  n11 = Subtract(A=Add(A=@n9.R, B=ScalarParameter() {defaultValue=2000, parameterName=Wave Z}), B=@n10.R) {desc=Water Height This Frame}
  n12 = ScalarParameter() {defaultValue=100, parameterName=Depth for Fluid Interface}
  n13 = Subtract(A=Saturate(None=Divide(A=@n11, B=@n12))) {constB=0.5}
  n14 = Saturate(None=Multiply(A=Abs(None=@n13)) {constB=2})
  n15 = ScalarParameter() {defaultValue=32, parameterName=asdfdf}
  n16 = Divide(A=@n9.R, B=@n15)
  n17 = Multiply(A=@n16) {constB=-1}
  n18 = TextureCoordinate()
  n19 = Sign(None=Subtract(A=ComponentMask(None=@n18) {g=true}) {constB=0.5})
  n20 = Multiply(A=Constant2Vector() {g=1}, B=Divide() {constB=1024})
  n21 = TextureSampleParameter2D(UVs=Add(A=@n8, B=Multiply(A=@n19, B=@n20))) {desc="Wave ", parameterName=WaveRT, samplerType=SAMPLERTYPE_LinearColor, texture=&RT_A}
  n22 = Divide(A=Subtract(A=@n9.R, B=@n21.R), B=@n15)
  Multiply(A=SquareRoot(None=OneMinus(None=@n14)), B=Saturate(None=Multiply(A=@n17, B=@n22)))
  Multiply(B=@n19)
  n23 = Subtract(A=@n11, B=ScalarParameter() {defaultValue=500, parameterName=Rise of Grid}) {constB=10}
  Saturate(None=Multiply(A=Min(A=@n23) {constB=0}) {constB=-1})
  Saturate(None=@n16)
  Subtract(A=@n4.R, B=@n1.R)
  Subtract(A=Add(A=@n21.R, B=ScalarParameter() {defaultValue=2000, parameterName=Wave Z}), B=@n10.R) {desc=Water Height Last Frame}
  TextureCoordinate()
 ]
}
