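# Unlit utility material that renders Gerstner wave height as a heightmap, allowing negative emissive values.
# Centered UVs scaled by World Size form positions for MF_Gerstner_16_Waves, which evaluates waves for a water body index parameter.
# The Z component of the wave offset is divided by Divisor and multiplied by a SphereMask on the V coordinate alone, fading it away from the middle.
# A separate Blob multiply of that mask feeds no other listed node.
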Material /FortWater/FluidSimulation/Materials/Utility/M_Render_Wave_Heightmap
graph {
 properties {bAllowNegativeEmissiveColor=true, pixelDepthOffsetMode=PDOM_Legacy, shadingModel=MSM_Unlit, usedShadingModels=Unlit}
 nodes [
  n1 = SphereMask(
   A=ComponentMask(None=TextureCoordinate()) {g=true}
   B=Constant() {r=0.5}
   Radius=ScalarParameter() {defaultValue=1, parameterName=R}
   Hardness=ScalarParameter() {defaultValue=0.9, parameterName=H}
  ) {hardnessPercent=0.5}
  Multiply(A=@n1, B=ScalarParameter() {parameterName=Blob})
  n2 = Subtract(A=TextureCoordinate()) {constB=0.5}
  n3 = ScalarParameter() {defaultValue=51000, parameterName=World Size}
  n4 = MaterialFunctionCall!(
   Z Offset Only
   Position=AppendVector(A=Multiply(A=@n2, B=@n3), B=Constant())
   WaterBodyIndex=ScalarParameter() {defaultValue=1, parameterName=Index}
   Normals use VS Interpolation=StaticBool()
   All Normals VS
  ) {materialFunction=/FortWater/Materials/Functions/MF_Gerstner_16_Waves}
  Multiply(A=Divide(A=ComponentMask(None=@n4.WPO) {b=true}, B=ScalarParameter() {defaultValue=1, parameterName=Divisor}), B=@n1)
 ]
}
