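# Unlit utility material for the water fluid simulation that builds a binary edge mask, marking where captured height lies above the water surface.
# It converts the RT texture's red channel to height by subtracting it from CaptureZ.
# UseTerrainSystem blends the reference height from a constant WaterZ to WaterBodyData's water height, sampled at the render target's world position.
# Ceil and Saturate turn the height difference into a hard 0 or 1.
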Material /Water/FluidSimulation/Materials/Utility/MakeEdgeMask
graph {
 properties {pixelDepthOffsetMode=PDOM_Legacy, shadingModel=MSM_Unlit, usedShadingModels=Unlit}
 nodes [
  n1 = TextureSampleParameter2D() {
   parameterName=RT
   samplerType=SAMPLERTYPE_LinearColor
   texture=/Engine/Functions/Engine_MaterialFunctions02/PivotPainter2/Black_1x1_EXR_Texture
  }
  n2 = Subtract(A=ScalarParameter() {defaultValue=50000, parameterName=CaptureZ}, B=@n1.R)
  n3 = LinearInterpolate(
   A=ScalarParameter() {defaultValue=-1536, parameterName=WaterZ}
   B=MaterialFunctionCall!(Position=MaterialFunctionCall!(UV) {materialFunction=/Water/FluidSimulation/Materials/Functions/WaveFoamSim_RenderTargetWorldPosition}.WorldPosition) {materialFunction=/Water/Materials/Functions/WaterBodyData}.Water Z Height
   Alpha=ScalarParameter() {parameterName=UseTerrainSystem}
  )
  Saturate(None=Ceil(None=Subtract(A=@n2, B=@n3)))
 ]
}
